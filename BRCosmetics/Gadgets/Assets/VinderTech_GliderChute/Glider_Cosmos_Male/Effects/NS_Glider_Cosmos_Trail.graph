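# Niagara engine-thruster trail for the Cosmos male glider chute: mesh flames, sparks, an acceleration glow and flickering warm lights, attached to two skeletal mesh bones.
# User.ForwardDot, overridden through front-end modifiers for menu display, drives spawn, lifetime, brightness and light response.
# A system-level forward-burst value adds extra flame and spark spawning and stretches the flame meshes.
# Counter-based spawn groups and placement pickers distribute particles across bones and offsets; Smoke and NE_Empty emitters are disabled.
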NiagaraSystem /BRCosmetics/Gadgets/Assets/VinderTech_GliderChute/Glider_Cosmos_Male/Effects/NS_Glider_Cosmos_Trail
dependencies {
 usedDataInterfaces [
  &NiagaraDataInterfaceColorCurve
  &NiagaraDataInterfaceCurve
  &NiagaraDataInterfaceVector2DCurve
  /Script/Niagara.NiagaraDataInterfaceSimpleCounter
  &NiagaraDataInterfaceArrayFloat3
  &NiagaraDataInterfaceVectorCurve
 ]
 usedDynamicInputs []
 usedModules [
  &self:SystemSpawnScript.SystemScriptSource.SystemScriptGraph.NiagaraNodeAssignment_0.SetVariables_#1
  &SystemState
  &FN_Integrator
  &self:SystemSpawnScript.SystemScriptSource.SystemScriptGraph.NiagaraNodeAssignment_1.SetVariables_#2
 ]
}
emitters {
 AccelerationGlow {
  emitterUpdateScript [
   &EmitterState()
   &SpawnRate(
    Spawn Probability=&FN_FrontEndFloatModifier(Gameplay=$User.ForwardDot, Front End=1)
    SpawnRate=64
    Use Spawn Probability=true
   )
  ]
  particleSpawnScript [
   /SetVariables(Particles.SpawnGroup=&ModuloInt(
     A=/Niagara/DynamicInputs/DataInterface/GenerateIntFromCounter(SimpleCounter:NiagaraDataInterfaceSimpleCounter=DataInterface() {propertyValues {GpuSyncMode=None, InitialValue=0}})
     B=2
    )) {isSetParameters=true, module=SetVariables_#3}
   /Niagara/Modules/Spawn/Initialization/InitializeParticle(
    Color=<rgba 1 1 1 0.1>
    Lifetime=0.25
    Module.Write Color=true
    Module.Write Lifetime=true
    Module.Write Mass=true
    Module.Write Material Random=true
    Module.Write Position=false
    Module.Write Scale=false
    Module.Write SpriteRotation=false
    Module.Write SpriteSize=true
    Position=&SimulationPosition()
    Sprite Size=&Vector2DFromFloat(Value=200)
   )
   &DynamicMaterialParameters(
    Index 0 Param 1=100
    Module.Param0WriteEnabled=true
    Module.Param1WriteEnabled=false
    Module.Param2WriteEnabled=false
    Module.Param3WriteEnabled=false
   )
   &CameraOffset(Camera Offset Amount=29)
  ]
  particleUpdateScript [
   &ParticleState()
   &SkeletalMeshLocation(
    Bone / Socket Index=$Particles.SpawnGroup
    Bone Sampling Mode=::NewEnumerator4
    Inherit Velocity=false
    Interpolate Mesh Sampling=false
    MaskSpawnGroup=false
    Sampled Position Offset=&Add_Vector(
     A=$System.OffsetLoopFromEngineBone
     B=&FN_FrontEndVector3Modifier(Gameplay=<0 -5 0>, Front End=<0 10 0>)
    )
    Sampled Position Offset Coordinate Space:ENiagaraMeshTransforms=::NewEnumerator3
    Skeletal Mesh=$User.Skeletal Mesh
   )
   &ScaleColor(
    Scale Alpha=/Niagara/DynamicInputs/Math/LerpMultipleFloats(
     Float Count:ENiagara_MultipleLerpCount=::NewEnumerator2
     1:NiagaraFloat=0
     2:NiagaraFloat=0.2
     3:NiagaraFloat=0.5
     4:NiagaraFloat=1
     5:NiagaraFloat=0
     Alpha:NiagaraFloat=$Particles.NormalizedAge
    )
    Scale RGB=&VectorFromCurve(
     VectorCurve=DataInterface() {propertyValues {
       XCurve {keys [
         {interpMode=RCIM_Cubic, tangentMode=RCTM_User, time=0, value=1}
         {arriveTangent=12.4991, interpMode=RCIM_Cubic, leaveTangent=12.4991, tangentMode=RCTM_User, time=1, value=8}
        ]}
       YCurve {keys [{interpMode=RCIM_Cubic, tangentMode=RCTM_User, time=0, value=1}, <at 1 1>]}
       ZCurve {keys [
         {arriveTangent=-2.6029, interpMode=RCIM_Cubic, leaveTangent=-2.6029, tangentMode=RCTM_User, time=-0.0172, value=0.9325}
         {arriveTangent=-0.6001, interpMode=RCIM_Cubic, leaveTangent=-0.6001, tangentMode=RCTM_User, time=1, value=0}
        ]}
      }}
     Scale Curve=&VectorFromFloat(Value=&FN_FrontEndFloatModifier(Gameplay=1, Front End=0.35))
    )
    ScaleA=true
    ScaleRGB=true
   )
   &ScaleSpriteSize(Scale Factor=&Vector2DFromFloat(Value=&Lerp_Float(A=0, B=1, Alpha=$Particles.NormalizedAge)))
  ]
  rendererClasses [&NiagaraSpriteRendererProperties]
  renderers [{rendererClass=&NiagaraSpriteRendererProperties}]
 }
 Flame_Mesh {
  emitterUpdateScript [
   &EmitterState()
   &SpawnRate(
    SpawnRate=&Lerp_Float(A=16, B=32, Alpha=&FN_FrontEndFloatModifier(Gameplay=$User.ForwardDot, Front End=1))
    Use Spawn Probability=false
   )
   &SpawnRate(Spawn Probability=$System.GenerateForwardBurst, SpawnRate=140, Use Spawn Probability=true) {module=SpawnRate001}
  ]
  particleSpawnScript [
   /SetVariables(
    Particles.PlacementPicker:NiagaraInt32=/Niagara/DynamicInputs/Divide/Divide_Int(
     A:NiagaraInt32=&ModuloInt(
      A=&GenerateIntFromCounter(SimpleCounter:NiagaraDataInterfaceSimpleCounter=DataInterface() {propertyValues {GpuSyncMode=None, InitialValue=0}})
      B=4
     )
     B:NiagaraInt32=2
    )
    Particles.SpawnGroup=&ModuloInt(
     A=&GenerateIntFromCounter(SimpleCounter:NiagaraDataInterfaceSimpleCounter=DataInterface() {propertyValues {GpuSyncMode=None, InitialValue=0}})
     B=2
    )
   ) {isSetParameters=true}
   /SetVariables(Particles.ShowBackFire:NiagaraInt32=/Niagara/DynamicInputs/Subtract/Subtract_Integer(A:NiagaraInt32=1, B:NiagaraInt32=$Particles.PlacementPicker)) {isSetParameters=true}
   &InitializeParticle(
    Lifetime=&Lerp_Float(A=1, B=0.5, Alpha=&FN_FrontEndFloatModifier(Gameplay=$User.ForwardDot, Front End=1))
    Mesh Scale=&Multiply_Vector(
     A=&SelectVectorFromArray(
      Direct Array Index=$Particles.PlacementPicker
      Vector Selection Array=DataInterface() {propertyValues {FloatData [<0.15 0.15 0.04>, <0.23 0.23 0.06>]}}
     )
     B=&Lerp_Vector(
      A=<1 1 1>
      B=&RandomRangeVector(Minimum=<1 1 1>, Maximum=<1 1 1.5>)
      Alpha=&FN_FrontEndFloatModifier(Gameplay=$User.ForwardDot, Front End=1)
     )
    )
    Module.Write Color=true
    Module.Write Lifetime=true
    Module.Write Mass=false
    Module.Write Material Random=false
    Module.Write Position=false
    Module.Write Scale=true
    Module.Write SpriteRotation=false
    Module.Write SpriteSize=false
    Position=&SimulationPosition()
    Sprite Size=<5 5>
   )
   &DynamicMaterialParameters(
    Index 0 Param 1=&RandomRangeFloat(Minimum=0, Maximum=1)
    Index 0 Param 4=&Lerp_Float(A=2, B=3, Alpha=&FN_FrontEndFloatModifier(Gameplay=$User.ForwardDot, Front End=1))
    Module.Param0WriteEnabled=true
    Module.Param1WriteEnabled=false
    Module.Param2WriteEnabled=false
    Module.Param3WriteEnabled=true
   )
   /Niagara/Modules/Spawn/Orientation/V2/InitialMeshOrientation(
    Orientation Axis:Vector3f=<0 -1 0>
    Orientation Vector=<0 0 1>
    Rotation=&RandomRangeVector(Minimum=<0 0 0>, Maximum=<0 0 1>)
    Use Orientation Vector=true
    Use Rotation Vector=true
   )
  ]
  particleUpdateScript [
   &ParticleState()
   &SkeletalMeshLocation(
    Apply Sampled Orientation to Mesh Particles=true
    Bone / Socket Index=$Particles.SpawnGroup
    Bone Sampling Mode=::NewEnumerator4
    Inherit Velocity=false
    Interpolate Mesh Sampling=false
    MaskSpawnGroup=false
    Sampled Position Offset=&Lerp_Vector(
     A=$System.OffsetLoopFromEngineBone
     B=&Add_Vector(A=$System.OffsetLoopFromEngineBone, B=<0 -16 0>)
     Alpha=&MakeFloatFromInt(Int=$Particles.PlacementPicker)
    )
    Sampled Position Offset Coordinate Space:ENiagaraMeshTransforms=::NewEnumerator3
    Skeletal Mesh=$User.Skeletal Mesh
   )
   &ScaleMeshSize(Scale Factor=&VectorFromCurve(
     VectorCurve=DataInterface() {propertyValues {
       XCurve {keys [<at 0 1>,<at 1 1>]}
       YCurve {keys [<at 0 1>,<at 1 1>]}
       ZCurve {keys [<at 0.9 1 RCIM_Cubic>, {arriveTangent=-4.873, interpMode=RCIM_Cubic, leaveTangent=-4.873, tangentMode=RCTM_User, time=1, value=0}]}
      }}
     Scale Curve=&Lerp_Vector(
      A=<1 1 1>
      B=<1 1 2>
      Alpha=&self:InterpolateFloat(Float to Interpolate:NiagaraFloat=$System.GenerateForwardBurst, Initial Value:NiagaraFloat=1, Rate Of Change:NiagaraFloat=20)
     )
    ))
   &ScaleColor(
    Scale Alpha=&Lerp_Float(
     A=0
     B=&Lerp_Float(A=2, B=4, Alpha=&FN_FrontEndFloatModifier(Gameplay=$User.ForwardDot, Front End=1))
     Alpha=$Particles.NormalizedAge
    )
    Scale RGB=&Lerp_Vector(A=<1 1 1>, B=<4 5 5>, Alpha=&FN_FrontEndFloatModifier(Gameplay=$User.ForwardDot, Front End=1))
    ScaleA=true
    ScaleRGB=true
   )
  ]
  rendererClasses [&NiagaraMeshRendererProperties]
  renderers [
   {rendererClass=&NiagaraMeshRendererProperties, rendererIndex=0}
   {rendererClass=&NiagaraMeshRendererProperties, rendererIndex=1}
  ]
 }
 Light {
  emitterUpdateScript [
   &EmitterState()
   &SpawnRate(SpawnRate=4, Use Spawn Probability=false) {enabled=false}
   &SpawnBurst_Instantaneous(
    Spawn Count=4
    Spawn Probability=&MakeCustomFloatFromBool(
     Bool=&SetBoolByIntComparison(A=$Emitter.LoopCount, B=0, Comparison Type=::NewEnumerator0)
     True Float=0
     False Float=1
    )
    Use Spawn Probability=true
   )
  ]
  particleSpawnScript [
   /SetVariables(
    Particles.PlacementPicker:NiagaraInt32=&Divide_Int(
     A:NiagaraInt32=&ModuloInt(
      A=&GenerateIntFromCounter(SimpleCounter:NiagaraDataInterfaceSimpleCounter=DataInterface() {propertyValues {GpuSyncMode=None, InitialValue=0}})
      B=4
     )
     B:NiagaraInt32=2
    )
    Particles.SpawnGroup=&ModuloInt(
     A=&GenerateIntFromCounter(SimpleCounter:NiagaraDataInterfaceSimpleCounter=DataInterface() {propertyValues {GpuSyncMode=None, InitialValue=0}})
     B=2
    )
   ) {isSetParameters=true, module=SetVariables_#3}
   &InitializeParticle(
    Color=<rgba 20 6.5 2 0.5>
    Module.Write Color=true
    Module.Write Lifetime=true
    Module.Write Mass=true
    Module.Write Material Random=true
    Module.Write Position=false
    Module.Write Scale=false
    Module.Write SpriteRotation=false
    Module.Write SpriteSize=true
    Position=&SimulationPosition()
    Sprite Size=<4 4>
   )
   &CameraOffset(Camera Offset Amount=3)
  ]
  particleUpdateScript [
   &ParticleState(Kill Particles When Lifetime Has Elapsed=false, Let Infinitely Lived Particles Die When Emitter Deactivates=true)
   &SkeletalMeshLocation(
    Bone / Socket Index=$Particles.SpawnGroup
    Bone Sampling Mode=::NewEnumerator4
    Inherit Velocity=false
    Interpolate Mesh Sampling=false
    MaskSpawnGroup=false
    Sampled Position Offset=&Add_Vector(
     A=&Lerp_Vector(
      A=$System.OffsetLoopFromEngineBone
      B=&Add_Vector(A=$System.OffsetLoopFromEngineBone, B=<0 -18 0>)
      Alpha=&MakeFloatFromInt(Int=$Particles.PlacementPicker)
     )
     B=<0 7 0>
    )
    Sampled Position Offset Coordinate Space:ENiagaraMeshTransforms=::NewEnumerator3
    Skeletal Mesh=$User.Skeletal Mesh
   )
   &FN_Light_v3(
    Color=&LerpLinearColors(
     StartColor=<rgba 150 48.75 15 1>
     EndColor=<rgba 300 135.075 80.1 1>
     LerpFactor=&FN_FrontEndFloatModifier(Gameplay=$User.ForwardDot, Front End=1)
    )
    Create Color Attribute=true
    Create Exponent Attribute=true
    Create Radius Attribute=true
    Exponent:NiagaraFloat=&Lerp_Float(A=1, B=2, Alpha=&MakeFloatFromInt(Int=$Particles.PlacementPicker))
    Radius:NiagaraFloat=&self:InterpolateFloat(
     Float to Interpolate:NiagaraFloat=&Lerp_Float(
      A=&RandomRangeFloat(
       Minimum=17
       Maximum=20
       Evaluation Type=::NewEnumerator1
       Randomness Mode=::NewEnumerator1
       Random Seed=&Add_Integer(A=$Particles.SpawnGroup, B=&Add_Integer(A=$Emitter.LoopCount, B=$System.LoopCount))
       Fixed Random Seed=true
      )
      B=&RandomRangeFloat(
       Minimum=25
       Maximum=30
       Evaluation Type=::NewEnumerator1
       Randomness Mode=::NewEnumerator1
       Random Seed=&Add_Integer(A=$Particles.SpawnGroup, B=&Add_Integer(A=$Emitter.LoopCount, B=$System.LoopCount))
       Fixed Random Seed=true
      )
      Alpha=&MakeFloatFromInt(Int=$Particles.PlacementPicker)
     )
     Initial Value:NiagaraFloat=0
     Rate Of Change:NiagaraFloat=&Lerp_Float(A=3, B=10, Alpha=&FN_FrontEndFloatModifier(Gameplay=$User.ForwardDot, Front End=1))
    )
    Scale Color:NiagaraFloat=&self:InterpolateFloat(
     Float to Interpolate:NiagaraFloat=&RandomRangeFloat(Minimum=0.25, Maximum=1, Evaluation Type=::NewEnumerator1)
     Initial Value:NiagaraFloat=0
     Rate Of Change:NiagaraFloat=&Lerp_Float(A=25, B=100, Alpha=&FN_FrontEndFloatModifier(Gameplay=$User.ForwardDot, Front End=1))
    )
   )
   &FN_Light_v3(
    Color=<rgba 200 65 20 1>
    Create Color Attribute=true
    Create Exponent Attribute=true
    Create Radius Attribute=true
    Enable Particle Light Enable=true
    Exponent=1
    Light Enabled=true
    Radius:NiagaraFloat=&Lerp_Float(
     A=&RandomRangeFloat(
      Minimum=17
      Maximum=20
      Evaluation Type=::NewEnumerator1
      Randomness Mode=::NewEnumerator1
      Random Seed=&Add_Integer(A=$Particles.SpawnGroup, B=&Add_Integer(A=$Emitter.LoopCount, B=$System.LoopCount))
      Fixed Random Seed=true
     )
     B=&RandomRangeFloat(
      Minimum=25
      Maximum=25
      Evaluation Type=::NewEnumerator1
      Randomness Mode=::NewEnumerator1
      Random Seed=&Add_Integer(A=$Particles.SpawnGroup, B=&Add_Integer(A=$Emitter.LoopCount, B=$System.LoopCount))
      Fixed Random Seed=true
     )
     Alpha=&MakeFloatFromInt(Int=$Particles.PlacementPicker)
    )
    Scale Color:NiagaraFloat=&RandomRangeFloat(Minimum=0.25, Maximum=1, Evaluation Type=::NewEnumerator1)
   ) {enabled=false, module=FN_Light_v3001}
  ]
  rendererClasses [&NiagaraSpriteRendererProperties, &NiagaraLightRendererProperties]
  renderers [
   {rendererClass=&NiagaraSpriteRendererProperties}
   {rendererClass=&NiagaraLightRendererProperties, rendererIndex=1}
  ]
 }
 NE_Empty {
  bEnabled=false
  emitterUpdateScript [
   &EmitterState(Life Cycle Mode=::NewEnumerator1, Loop Duration=1, UseLoopDelay=false)
   &SpawnRate(SpawnRate=20, Use Spawn Probability=false)
  ]
  particleSpawnScript [&InitializeParticle(
    Lifetime=5
    Module.Write Color=true
    Module.Write Lifetime=true
    Module.Write Mass=true
    Module.Write Material Random=true
    Module.Write Position=false
    Module.Write Scale=false
    Module.Write SpriteRotation=false
    Module.Write SpriteSize=true
    Position=&SimulationPosition()
    Sprite Size=<500 500>
   )]
  particleUpdateScript [
   &ParticleState()
   &SolveForcesAndVelocity(Clamp Velocity=false, Limit Acceleration=false)
  ]
  rendererClasses [&NiagaraSpriteRendererProperties]
  renderers [{rendererClass=&NiagaraSpriteRendererProperties}]
 }
 Smoke {
  bEnabled=false
  emitterUpdateScript [
   &EmitterState()
   /SetVariables(
    Emitter.ForwardDotInterp:NiagaraFloat=&self:InterpolateFloat(Float to Interpolate:NiagaraFloat=$User.ForwardDot, Initial Value:NiagaraFloat=0, Rate Of Change:NiagaraFloat=3)
    Emitter.RightDotInterp:NiagaraFloat=&self:InterpolateFloat(Float to Interpolate:NiagaraFloat=$User.RightDot, Initial Value:NiagaraFloat=0, Rate Of Change:NiagaraFloat=3)
   ) {isSetParameters=true}
   &SpawnRate(
    Spawn Probability=&MakeCustomFloatFromBool(
     Bool=&Boolean_AndOperation(
      Bool1=&Boolean_AndOperation(
       Bool1=&SetBoolByFloatComparison(A=$Emitter.RightDotInterp, B=0.01, Comparison Type=::NewEnumerator6)
       Bool2=&SetBoolByFloatComparison(A=$Emitter.RightDotInterp, B=-0.01)
      )
      Bool2=&Boolean_AndOperation(
       Bool1=&Boolean_NotOperation(Bool=$User.bIsFrontEndPreview)
       Bool2=&SetBoolByFloatComparison(A=$Emitter.ForwardDotInterp, B=0.01, Comparison Type=::NewEnumerator6)
      )
     )
     True Float=1
     False Float=0
    )
    SpawnRate=300
    Use Spawn Probability=true
   ) {module=SpawnRate001}
  ]
  particleSpawnScript [
   &InitializeParticle(
    Color=&UniformRangedLinearColor(Minimum=<rgba 0.505 0.505 0.505 0.005>, Maximum=<rgba 0.771 0.771 0.771 0.01>)
    Lifetime=0.3
    Module.Write Color=true
    Module.Write Lifetime=true
    Module.Write Mass=true
    Module.Write Material Random=true
    Module.Write Position=false
    Module.Write Scale=false
    Module.Write SpriteRotation=true
    Module.Write SpriteSize=true
    Position=&SimulationPosition()
    Sprite Rotation=&UniformRangedFloat(Minimum=0, Maximum=360)
    Sprite Size=&Vector2DFromFloat(Value=&UniformRangedFloat(Minimum=20, Maximum=40))
   )
   &ApplyInitialForces()
   /SetVariables(Particles.PickBone:NiagaraInt32=&RandomRangeInt(Minimum=0, Maximum=1)) {isSetParameters=true}
   &SkeletalMeshLocation(
    Bone / Socket Index=$Particles.PickBone
    Bone Sampling Mode=::NewEnumerator4
    Inherit Velocity=false
    MaskSpawnGroup=false
    Sampled Position Offset=&Lerp_Vector(
     A=&RandomRangeVector(Minimum=<0 -50 -30>, Maximum=<0 -20 -30>)
     B=<0 -35 -30>
     Alpha=&MakeFloatFromInt(Int=$Particles.SpriteVisibility)
    )
    Sampled Position Offset Coordinate Space:ENiagaraMeshTransforms=::NewEnumerator3
    Skeletal Mesh=$User.Skeletal Mesh
   )
   &AddVelocity(Velocity=<0 0 -1000>)
  ]
  particleUpdateScript [
   &InheritVelocity(
    Module.Limit Speed=false
    Velocity Scale=&Multiply_VectorByFloat(Vector=<0.1 0.1 0>, Float=&Lerp_Float(A=1, B=0, Alpha=$Particles.NormalizedAge))
   )
   &ParticleState()
   &SolveForcesAndVelocity(Clamp Velocity=false, Limit Acceleration=false)
   &ScaleSpriteSize(Scale Factor=&Vector2DFromFloat(Value=&FloatFromCurve(FloatCurve=DataInterface() {propertyValues {Curve {keys [
          {arriveTangent=3.1801, interpMode=RCIM_Cubic, leaveTangent=3.1801, tangentMode=RCTM_User, time=0, value=0}
          {interpMode=RCIM_Cubic, tangentMode=RCTM_User, time=0.5164, value=1}
          {arriveTangent=-3.6914, interpMode=RCIM_Cubic, leaveTangent=-3.6914, tangentMode=RCTM_User, time=1, value=0}
         ]}}})))
   /SetVariables(Particles.SubImageIndex=&FloatFromCurve(FloatCurve=DataInterface() {propertyValues {Curve {keys [
         {arriveTangent=29.8623, interpMode=RCIM_Cubic, leaveTangent=29.8623, tangentMode=RCTM_User, time=-0.0046, value=5.0556}
         <at 1.0011 22.1689 RCIM_Cubic>
        ]}}})) {isSetParameters=true}
   /SetVariables(Particles.SpriteDynaMatParams:Vector4f=&MakeVector4(
     X=&FloatFromCurve(FloatCurve=DataInterface() {propertyValues {Curve {keys [<at -0.0046 1.0113 RCIM_Cubic>,<at 1.0011 0.002>]}}})
     Y=0
     Z=0
     W=&FloatFromCurve(FloatCurve=DataInterface() {propertyValues {Curve {keys [<at 0 0 RCIM_Cubic>,<at 1 0.5>]}}})
    )) {isSetParameters=true}
   &ScaleColor(
    Scale Alpha=&FloatFromCurve(FloatCurve=DataInterface() {propertyValues {Curve {keys [
         {arriveTangent=-3.1757, interpMode=RCIM_Cubic, leaveTangent=-3.1757, tangentMode=RCTM_User, time=0.5, value=1.0113}
         {arriveTangent=-1.5121, interpMode=RCIM_Cubic, leaveTangent=-1.5121, tangentMode=RCTM_User, time=1.0011, value=0.002}
        ]}}})
    ScaleA=true
    ScaleRGB=false
   )
  ]
  rendererClasses [&NiagaraSpriteRendererProperties]
  renderers [{rendererClass=&NiagaraSpriteRendererProperties}]
 }
 Sparks {
  emitterUpdateScript [
   &EmitterState()
   &SpawnRate(
    Spawn Probability=&Lerp_Float(
     A=0.1
     B=&RandomRangeFloat(Minimum=0.05, Maximum=0.2)
     Alpha=&FN_FrontEndFloatModifier(Gameplay=$User.ForwardDot, Front End=1)
    )
    SpawnRate=50
    Use Spawn Probability=true
   )
   &SpawnRate(Spawn Probability=$System.GenerateForwardBurst, SpawnRate=200, Use Spawn Probability=true) {module=SpawnRate001}
  ]
  particleSpawnScript [&InitializeParticle(
    Color=<rgba 5 2.625 1.25 1>
    Lifetime=&RandomRangeFloat(Minimum=0.1, Maximum=0.2)
    Module.Write Color=true
    Module.Write Lifetime=true
    Module.Write Mass=true
    Module.Write Material Random=true
    Module.Write Position=false
    Module.Write Scale=false
    Module.Write SpriteRotation=false
    Module.Write SpriteSize=true
    Position=&SimulationPosition()
    Sprite Size=&Lerp_Vector2(
     A=&UniformRangedVector2D(Minimum=<0.5 1>, Maximum=<1 2>)
     B=&UniformRangedVector2D(Minimum=<1 1.5>, Maximum=<2 4>)
     Alpha=&FN_FrontEndFloatModifier(Gameplay=$User.ForwardDot, Front End=1)
    )
   )]
  particleUpdateScript [
   &ParticleState()
   /SetVariables(Particles.StartPosition:Vector3f=&Lerp_Vector(
     A=&Add_Vector(
      A=<0 -16 0>
      B=&Multiply_VectorByFloat(
       Vector=/Niagara/DynamicInputs/Helpers/SafeNormalizeVector(
        Vector To Normalize:Vector3f=&RandomRangeVector(Minimum=<-1 0 -1>, Maximum=<1 0 1>)
        Fallback Vector:Vector3f=<0 0 0>
       )
       Float=22
      )
     )
     B=&Multiply_VectorByFloat(
      Vector=&SafeNormalizeVector(
       Vector To Normalize:Vector3f=&RandomRangeVector(Minimum=<-1 0 -1>, Maximum=<1 0 1>)
       Fallback Vector:Vector3f=<0 0 0>
      )
      Float=12
     )
     Alpha=&UniformAOrBFloat(A=0)
    )) {isSetParameters=true}
   /SetVariables(Particles.EndPosition:Vector3f=&Add_Vector(
     A=&Multiply_Vector(A=$Particles.StartPosition, B=<1.5 1 1.5>)
     B=&RandomRangeVector(Minimum=<0 50 0>, Maximum=<0 250 0>)
    )) {isSetParameters=true}
   /SetVariables(Particles.SpriteAlignment=&FN_FrontEndVector3Modifier(
     Gameplay=&Subtract_Vector(A=$Particles.EndPosition, B=$Particles.StartPosition)
     Front End=/Game/Effects/Niagara/DynamicInputs/FrontEnd/FN_FrontEndRotateVector(
      Angle:NiagaraFloat=0.2
      Vector to Rotate:Vector3f=&Subtract_Vector(A=$Particles.EndPosition, B=$Particles.StartPosition)
     )
    )) {isSetParameters=true}
   &SkeletalMeshLocation(
    Bone / Socket Index=&RandomRangeInt(Minimum=0, Maximum=1)
    Bone Sampling Mode=::NewEnumerator4
    Inherit Velocity=false
    Interpolate Mesh Sampling=false
    MaskSpawnGroup=false
    Sampled Position Offset=&Add_Vector(
     A=$System.OffsetBurstFromEngineBone
     B=&Lerp_Vector(A=$Particles.StartPosition, B=$Particles.EndPosition, Alpha=$Particles.NormalizedAge)
    )
    Sampled Position Offset Coordinate Space:ENiagaraMeshTransforms=::NewEnumerator3
    Skeletal Mesh=$User.Skeletal Mesh
    Use Sampled Mesh Attribute as Ribbon ID=false
   ) {module=SkeletalMeshLocation001}
   &ScaleColor(
    Scale Alpha=&FloatFromCurve(
     FloatCurve=DataInterface() {propertyValues {Curve {keys [<at 0 1>,<at 1 0>]}}}
     CurveIndex=&Multiply_Float(A=$Particles.NormalizedAge, B=2)
    )
    Scale RGB=&MakeVectorFromLinearColorRGB(Color=&ColorFromCurve(
      ColorCurve=DataInterface() {propertyValues {
        AlphaCurve {keys []}
        BlueCurve {keys [<at 0 0.25 RCIM_Cubic>,<at 0.5 0.5>,<at 1 0.875 RCIM_Cubic>]}
        GreenCurve {keys [<at 0 0.525 RCIM_Cubic>,<at 0.5 0.6833>,<at 1 0.9208 RCIM_Cubic>]}
        RedCurve {keys [<at 0 1 RCIM_Cubic>,<at 0.5 1>,<at 1 1 RCIM_Cubic>]}
       }}
      CurveIndex=&Multiply_Float(A=$Particles.NormalizedAge, B=2)
     ))
    ScaleA=true
    ScaleRGB=true
   )
   &ScaleColor(
    Scale RGB=&VectorFromFloat(Value=&FloatFromCurve(
      FloatCurve=DataInterface() {propertyValues {Curve {keys [
          <at 0 100 RCIM_Cubic>
          {arriveTangent=-36.3579, interpMode=RCIM_Cubic, leaveTangent=-36.3595, tangentMode=RCTM_User, time=0.25, value=10}
          <at 1 1 RCIM_Cubic>
         ]}}}
      CurveIndex=&Multiply_Float(A=$Particles.NormalizedAge, B=2)
     ))
    ScaleA=true
    ScaleRGB=true
   ) {module=ScaleColor001}
   &ScaleSpriteSize(Scale Factor=&Vector2DFromCurve(
     Vector2Curve=DataInterface() {propertyValues {
       XCurve {keys [<at 0 1>,<at 1 0>]}
       YCurve {keys [{arriveTangent=18.2053, interpMode=RCIM_Cubic, leaveTangent=18.2053, tangentMode=RCTM_User, time=0, value=1}, <at 1 10>]}
      }}
     CurveIndex=&Multiply_Float(A=$Particles.NormalizedAge, B=2)
     ScaleCurve=<1 1>
    ))
   /SetVariables(
    Particles.ColorAlt:LinearColor=<rgba 0 0 0 0.015>
    Particles.PositionAlt:Vector3f=&Lerp_Vector(A=$Output.SkeletalMeshLocation001.SampledPosition, B=Linked() {linkedVariable {name=Particles.Position, type=NiagaraPosition}}, Alpha=0.4)
    Particles.SpriteFacingAlt:Vector3f=<0 0 1>
    Particles.SpriteRotationAlt:NiagaraFloat=&RandomRangeFloat(Minimum=-360, Maximum=360)
    Particles.SpriteSizeAlt:Vector2f=&Vector2DFromFloat(Value=&Multiply_Float(
      A=&RandomRangeFloat(Minimum=60, Maximum=90)
      B=&Lerp_Float(
       A=1
       B=&RandomRangeFloat(Minimum=0, Maximum=1)
       Alpha=&MakeFloatFromInt(Int=$Particles.SpawnGroup)
      )
     ))
   ) {isSetParameters=true}
  ]
  rendererClasses [&NiagaraSpriteRendererProperties]
  renderers [{rendererClass=&NiagaraSpriteRendererProperties}]
 }
}
summary {
 systemName=NS_Glider_Cosmos_Trail
 emitters [
  {emitterName=Sparks, rendererClasses [&NiagaraSpriteRendererProperties]}
  {emitterName=Flame_Mesh, rendererClasses [&NiagaraMeshRendererProperties]}
  {emitterName=Light, rendererClasses [&NiagaraSpriteRendererProperties, &NiagaraLightRendererProperties]}
  {bEnabled=false, emitterName=NE_Empty, rendererClasses [&NiagaraSpriteRendererProperties]}
  {emitterName=AccelerationGlow, rendererClasses [&NiagaraSpriteRendererProperties]}
  {bEnabled=false, emitterName=Smoke, rendererClasses [&NiagaraSpriteRendererProperties]}
 ]
 userVariables [
  User.bIsFrontEnd {type=NiagaraBool}
  User.bIsFrontEndPreview {type=NiagaraBool}
  User.ForwardDot {type=NiagaraFloat, defaultValue=0, description=""}
  User.RightDot {type=NiagaraFloat, defaultValue=0, description=""}
  User.Skeletal Mesh {
   type=NiagaraDataInterfaceSkeletalMesh
   defaultValue {
    struct=NiagaraExt_VariableValue_DataInterface
    value {dataInterface=&self:NiagaraDataInterfaceSkeletalMesh_0}
   }
  }
 ]
}
